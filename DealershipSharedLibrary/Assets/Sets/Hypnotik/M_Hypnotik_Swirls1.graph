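# Particle material for the Hypnotik set that draws animated swirl patterns with soft masks, colored by particle color.
# Sine/cosine twists of radius-warped UVs, radial coordinates and texture distortion build the masks, with a parameter switching to the swirl branch.
# Dynamic parameters drive brightness, color shifts, tiling, speed and rotation.
# It also includes depth fade, optional temporal dithering and a time-driven sine/cosine wobble for world position offset.
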Material /DealershipSharedLibrary/Assets/Sets/Hypnotik/M_Hypnotik_Swirls1
graph {
 expressions {
  MaterialExpressionAbs_3 {class=MaterialExpressionAbs, inputs [None {from=MaterialExpressionDotProduct_12, output=""}], properties {}}
  MaterialExpressionAbs_4 {class=MaterialExpressionAbs, inputs [None {from=MaterialExpressionNamedRerouteUsage_66, output=""}], properties {}}
  MaterialExpressionAdd_12 {
   class=MaterialExpressionAdd
   inputs [
    A {from=MaterialExpressionTextureCoordinate_7}
    B {from=""}
   ]
   properties {constB=-0.5}
  }
  MaterialExpressionAdd_22 {
   class=MaterialExpressionAdd
   inputs [
    A {from=MaterialExpressionAdd_23}
    B {from=""}
   ]
   properties {constB=-0.5}
  }
  MaterialExpressionAdd_23 {
   class=MaterialExpressionAdd
   inputs [
    A {from=MaterialExpressionRotator_6}
    B {from=MaterialExpressionMultiply_24}
   ]
  }
  MaterialExpressionAdd_25 {
   class=MaterialExpressionAdd
   inputs [
    A {from=MaterialExpressionMultiply_103}
    B {from=MaterialExpressionMultiply_105}
   ]
  }
  MaterialExpressionAdd_26 {
   class=MaterialExpressionAdd
   inputs [
    A {from=MaterialExpressionMaterialFunctionCall_11, output=Time}
    B {from=MaterialExpressionDivide_2}
   ]
  }
  MaterialExpressionAdd_31 {
   class=MaterialExpressionAdd
   inputs [
    A {from=MaterialExpressionNamedRerouteUsage_5}
    B {from=MaterialExpressionConstantClamp_35}
   ]
  }
  MaterialExpressionAdd_32 {
   class=MaterialExpressionAdd
   inputs [
    A {from=MaterialExpressionMultiply_81}
    B {from=MaterialExpressionMultiply_82}
   ]
  }
  MaterialExpressionAdd_33 {
   class=MaterialExpressionAdd
   inputs [
    A {from=MaterialExpressionMultiply_0}
    B {from=MaterialExpressionMultiply_2}
   ]
  }
  MaterialExpressionAdd_34 {
   class=MaterialExpressionAdd
   inputs [
    A {from=MaterialExpressionMaterialFunctionCall_5, output=Radial Coordinates}
    B {from=MaterialExpressionMaterialFunctionCall_5, output=Radial Coordinates}
   ]
  }
  MaterialExpressionAdd_35 {
   class=MaterialExpressionAdd
   inputs [
    A {from=MaterialExpressionMultiply_79}
    B {from=MaterialExpressionMultiply_74}
   ]
  }
  MaterialExpressionAdd_36 {
   class=MaterialExpressionAdd
   inputs [
    A {from=MaterialExpressionMultiply_77}
    B {from=MaterialExpressionMultiply_78}
   ]
  }
  MaterialExpressionAdd_6 {
   class=MaterialExpressionAdd
   inputs [
    A {from=MaterialExpressionNamedRerouteUsage_56}
    B {from=MaterialExpressionNamedRerouteUsage_51}
   ]
  }
  MaterialExpressionAdd_7 {
   class=MaterialExpressionAdd
   inputs [
    A {from=MaterialExpressionMultiply_98}
    B {from=MaterialExpressionMultiply_92}
   ]
  }
  MaterialExpressionAdd_8 {
   class=MaterialExpressionAdd
   inputs [
    A {from=MaterialExpressionCosine_7}
    B {from=MaterialExpressionNamedRerouteUsage_37}
   ]
  }
  MaterialExpressionAdd_9 {
   class=MaterialExpressionAdd
   inputs [
    A {from=MaterialExpressionMultiply_106}
    B {from=MaterialExpressionNamedRerouteUsage_36}
   ]
  }
  MaterialExpressionAppendVector_13 {
   class=MaterialExpressionAppendVector
   inputs [
    A {from=MaterialExpressionScalarParameter_20, output=""}
    B {from=MaterialExpressionConstant_23, output=""}
   ]
  }
  MaterialExpressionAppendVector_14 {
   class=MaterialExpressionAppendVector
   inputs [
    A {from=MaterialExpressionAdd_35, output=""}
    B {from=MaterialExpressionSubtract_10, output=""}
   ]
  }
  MaterialExpressionCameraVector_2 {class=MaterialExpressionCameraVectorWS, inputs [], properties {}}
  MaterialExpressionComponentMask_12 {class=MaterialExpressionComponentMask, inputs [None {from=MaterialExpressionTextureCoordinate_14}], properties {r=true}}
  MaterialExpressionComponentMask_4 {
   class=MaterialExpressionComponentMask
   inputs [None {from=MaterialExpressionTextureSampleParameter2D_0, output=RGB}]
   properties {g=true, r=true}
  }
  MaterialExpressionComponentMask_5 {class=MaterialExpressionComponentMask, inputs [None {from=MaterialExpressionLinearInterpolate_7}], properties {g=true, r=true}}
  MaterialExpressionConstant3Vector_2 {class=MaterialExpressionConstant3Vector, properties {constant=<rgba 0 0 1 0>}}
  MaterialExpressionConstantBiasScale_0 {class=MaterialExpressionConstantBiasScale, inputs [None {from=MaterialExpressionSaturate_2}], properties {bias=-0.001, scale=3}}
  MaterialExpressionConstantBiasScale_18 {class=MaterialExpressionConstantBiasScale, inputs [None {from=MaterialExpressionTransform_2}]}
  MaterialExpressionConstantBiasScale_19 {class=MaterialExpressionConstantBiasScale, inputs [None {from=MaterialExpressionOneMinus_7}], properties {bias=-0.8, scale=1.5}}
  MaterialExpressionConstantBiasScale_20 {class=MaterialExpressionConstantBiasScale, inputs [None {from=MaterialExpressionMultiply_69}], properties {bias=0.0001, scale=10}}
  MaterialExpressionConstantBiasScale_21 {class=MaterialExpressionConstantBiasScale, inputs [None {from=MaterialExpressionOneMinus_8}], properties {bias=-0.795, scale=4}}
  MaterialExpressionConstantBiasScale_22 {class=MaterialExpressionConstantBiasScale, inputs [None {from=MaterialExpressionSubtract_8}], properties {bias=-0.1}}
  MaterialExpressionConstantBiasScale_24 {class=MaterialExpressionConstantBiasScale, inputs [None {from=MaterialExpressionAdd_32}], properties {bias=0, scale=6.5}}
  MaterialExpressionConstantBiasScale_26 {class=MaterialExpressionConstantBiasScale, inputs [None {from=MaterialExpressionOneMinus_7}], properties {bias=-0.9, scale=1}}
  MaterialExpressionConstantBiasScale_31 {
   class=MaterialExpressionConstantBiasScale
   inputs [None {from=MaterialExpressionNamedRerouteUsage_30}]
   properties {bias=-0.0001, scale=1.75}
  }
  MaterialExpressionConstantBiasScale_32 {class=MaterialExpressionConstantBiasScale, inputs [None {from=MaterialExpressionNamedRerouteUsage_40}], properties {bias=0, scale=1.5}}
  MaterialExpressionConstantBiasScale_34 {class=MaterialExpressionConstantBiasScale, inputs [None {from=MaterialExpressionNamedRerouteUsage_35}], properties {bias=-0.82, scale=1}}
  MaterialExpressionConstantBiasScale_35 {class=MaterialExpressionConstantBiasScale, inputs [None {from=MaterialExpressionMultiply_100}], properties {bias=-0.99, scale=0.01}}
  MaterialExpressionConstantBiasScale_37 {class=MaterialExpressionConstantBiasScale, inputs [None {from=MaterialExpressionMultiply_102}], properties {bias=-0.9, scale=2}}
  MaterialExpressionConstantBiasScale_38 {class=MaterialExpressionConstantBiasScale, inputs [None {from=MaterialExpressionMultiply_90}], properties {bias=-0.005, scale=60}}
  MaterialExpressionConstantBiasScale_39 {class=MaterialExpressionConstantBiasScale, inputs [None {from=MaterialExpressionConstantClamp_56}], properties {bias=-0.0008, scale=40}}
  MaterialExpressionConstantBiasScale_40 {class=MaterialExpressionConstantBiasScale, inputs [None {from=MaterialExpressionSine_6}], properties {bias=-0.8005, scale=3}}
  MaterialExpressionConstantBiasScale_41 {
   class=MaterialExpressionConstantBiasScale
   inputs [None {from=MaterialExpressionNamedRerouteDeclaration_28}]
   properties {bias=-0.65, scale=16}
  }
  MaterialExpressionConstantBiasScale_7 {class=MaterialExpressionConstantBiasScale, inputs [None {from=MaterialExpressionAdd_6}], properties {bias=-0.0005, scale=150}}
  MaterialExpressionConstantBiasScale_8 {class=MaterialExpressionConstantBiasScale, inputs [None {from=MaterialExpressionNamedRerouteUsage_34}], properties {bias=-0.83, scale=50}}
  MaterialExpressionConstantBiasScale_9 {class=MaterialExpressionConstantBiasScale, inputs [None {from=MaterialExpressionAdd_36}], properties {bias=-0.99, scale=1.5}}
  MaterialExpressionConstantClamp_26 {class=MaterialExpressionSaturate, inputs [None {from=MaterialExpressionConstantBiasScale_22}]}
  MaterialExpressionConstantClamp_27 {class=MaterialExpressionSaturate, inputs [None {from=MaterialExpressionMultiply_63}]}
  MaterialExpressionConstantClamp_28 {class=MaterialExpressionSaturate, inputs [None {from=MaterialExpressionConstantBiasScale_19}]}
  MaterialExpressionConstantClamp_29 {class=MaterialExpressionSaturate, inputs [None {from=MaterialExpressionPower_2}]}
  MaterialExpressionConstantClamp_30 {class=MaterialExpressionSaturate, inputs [None {from=MaterialExpressionMultiply_65}]}
  MaterialExpressionConstantClamp_31 {class=MaterialExpressionSaturate, inputs [None {from=MaterialExpressionConstantBiasScale_20}]}
  MaterialExpressionConstantClamp_32 {class=MaterialExpressionSaturate, inputs [None {from=MaterialExpressionConstantBiasScale_21}]}
  MaterialExpressionConstantClamp_33 {class=MaterialExpressionSaturate, inputs [None {from=MaterialExpressionSubtract_9}]}
  MaterialExpressionConstantClamp_34 {class=MaterialExpressionSaturate, inputs [None {from=MaterialExpressionAbs_4}]}
  MaterialExpressionConstantClamp_35 {class=MaterialExpressionSaturate, inputs [None {from=MaterialExpressionConstantBiasScale_24}]}
  MaterialExpressionConstantClamp_36 {class=MaterialExpressionSaturate, inputs [None {from=MaterialExpressionConstantBiasScale_26}]}
  MaterialExpressionConstantClamp_44 {class=MaterialExpressionSaturate, inputs [None {from=MaterialExpressionSubtract_2}]}
  MaterialExpressionConstantClamp_46 {class=MaterialExpressionSaturate, inputs [None {from=MaterialExpressionAbs_3}]}
  MaterialExpressionConstantClamp_47 {class=MaterialExpressionSaturate, inputs [None {from=MaterialExpressionConstantBiasScale_31}]}
  MaterialExpressionConstantClamp_48 {class=MaterialExpressionSaturate, inputs [None {from=MaterialExpressionMultiply_88}]}
  MaterialExpressionConstantClamp_49 {class=MaterialExpressionSaturate, inputs [None {from=MaterialExpressionConstantBiasScale_7}]}
  MaterialExpressionConstantClamp_50 {class=MaterialExpressionSaturate, inputs [None {from=MaterialExpressionConstantBiasScale_8}]}
  MaterialExpressionConstantClamp_52 {class=MaterialExpressionSaturate, inputs [None {from=MaterialExpressionConstantBiasScale_35}]}
  MaterialExpressionConstantClamp_53 {class=MaterialExpressionSaturate, inputs [None {from=MaterialExpressionConstantBiasScale_40}]}
  MaterialExpressionConstantClamp_54 {class=MaterialExpressionSaturate, inputs [None {from=MaterialExpressionConstantBiasScale_37}]}
  MaterialExpressionConstantClamp_55 {class=MaterialExpressionSaturate, inputs [None {from=MaterialExpressionConstantBiasScale_38}]}
  MaterialExpressionConstantClamp_56 {class=MaterialExpressionSaturate, inputs [None {from=MaterialExpressionNamedRerouteUsage_55}]}
  MaterialExpressionConstantClamp_57 {class=MaterialExpressionSaturate, inputs [None {from=MaterialExpressionConstantBiasScale_39}]}
  MaterialExpressionConstantClamp_58 {class=MaterialExpressionSaturate, inputs [None {from=MaterialExpressionConstantBiasScale_32}]}
  MaterialExpressionConstantClamp_59 {class=MaterialExpressionSaturate, inputs [None {from=MaterialExpressionScalarParameter_8}]}
  MaterialExpressionConstantClamp_60 {class=MaterialExpressionSaturate, inputs [None {from=MaterialExpressionConstantBiasScale_41}]}
  MaterialExpressionConstant_0 {class=MaterialExpressionConstant, properties {r=0.5}}
  MaterialExpressionConstant_1 {class=MaterialExpressionConstant, properties {r=0.5}}
  MaterialExpressionConstant_23 {class=MaterialExpressionConstant, properties {r=0.25}}
  MaterialExpressionConstant_28 {class=MaterialExpressionConstant, properties {r=0.125}}
  MaterialExpressionConstant_33 {class=MaterialExpressionConstant, properties {r=0.6}}
  MaterialExpressionConstant_37 {class=MaterialExpressionConstant, properties {r=2.5}}
  MaterialExpressionConstant_38 {class=MaterialExpressionConstant, properties {r=2}}
  MaterialExpressionConvert_0 {
   class=MaterialExpressionConvert
   inputs [None {from=MaterialExpressionNamedRerouteUsage_11}]
   properties {
    nodeName=Break Float2
    convertInputs [{
      type=Vector2
      expressionInput {
       expression=&self:MaterialExpressionNamedRerouteUsage_11
       inputName=None
       mask=0
       maskA=0
       maskB=0
       maskG=0
       maskR=0
       outputIndex=0
      }
     }]
    convertOutputs [{type=Scalar}, {type=Scalar}]
   }
  }
  MaterialExpressionConvert_1 {
   class=MaterialExpressionConvert
   inputs [
    None {from=MaterialExpressionCosine_4}
    None {from=MaterialExpressionSine_4}
    None {from=""}
   ]
   properties {
    nodeName=Make Vector3
    convertInputs [
     {
      type=Scalar
      expressionInput {
       expression=&self:MaterialExpressionCosine_4
       inputName=None
       mask=0
       maskA=0
       maskB=0
       maskG=0
       maskR=0
       outputIndex=0
      }
     }
     {
      type=Scalar
      expressionInput {
       expression=&self:MaterialExpressionSine_4
       inputName=None
       mask=0
       maskA=0
       maskB=0
       maskG=0
       maskR=0
       outputIndex=0
      }
     }
     {type=Scalar, expressionInput {expression=null, inputName=None, mask=0, maskA=0, maskB=0, maskG=0, maskR=0, outputIndex=0}}
    ]
    convertOutputs [{type=Vector3}]
   }
  }
  MaterialExpressionConvert_2 {
   class=MaterialExpressionConvert
   inputs [None {from=MaterialExpressionMultiply_96}]
   properties {
    nodeName=Break Float2
    convertInputs [{
      type=Vector2
      expressionInput {
       expression=&self:MaterialExpressionMultiply_96
       inputName=None
       mask=0
       maskA=0
       maskB=0
       maskG=0
       maskR=0
       outputIndex=0
      }
     }]
    convertOutputs [{type=Scalar}, {type=Scalar}]
   }
  }
  MaterialExpressionCosine_4 {class=MaterialExpressionCosine, inputs [None {from=MaterialExpressionAdd_26, output=""}], properties {}}
  MaterialExpressionCosine_5 {class=MaterialExpressionCosine, inputs [None {from=MaterialExpressionMultiply_76, output=""}], properties {period=2}}
  MaterialExpressionCosine_6 {class=MaterialExpressionCosine, inputs [None {from=MaterialExpressionAdd_8, output=""}], properties {period=0.675}}
  MaterialExpressionCosine_7 {class=MaterialExpressionCosine, inputs [None {from=MaterialExpressionMultiply_95, output=""}], properties {period=0.695}}
  MaterialExpressionDivide_2 {
   class=MaterialExpressionDivide
   inputs [
    A {from=MaterialExpressionComponentMask_12}
    B {from=MaterialExpressionScalarParameter_17}
   ]
  }
  MaterialExpressionDotProduct_0 {
   class=MaterialExpressionDotProduct
   inputs [
    A {from=MaterialExpressionRotator_1, output=""}
    B {from=MaterialExpressionRotator_1, output=""}
   ]
   properties {}
  }
  MaterialExpressionDotProduct_11 {
   class=MaterialExpressionDotProduct
   inputs [
    A {from=MaterialExpressionAdd_25, output=""}
    B {from=MaterialExpressionMultiply_91, output=""}
   ]
   properties {}
  }
  MaterialExpressionDotProduct_12 {
   class=MaterialExpressionDotProduct
   inputs [
    A {from=MaterialExpressionMultiply_22, output=""}
    B {from=MaterialExpressionMultiply_22, output=""}
   ]
   properties {}
  }
  MaterialExpressionDotProduct_14 {
   class=MaterialExpressionDotProduct
   inputs [
    A {from=MaterialExpressionAdd_7, output=""}
    B {from=MaterialExpressionSubtract_1, output=""}
   ]
   properties {}
  }
  MaterialExpressionDotProduct_8 {
   class=MaterialExpressionDotProduct
   inputs [
    A {from=MaterialExpressionTransform_0, output=""}
    B {from=MaterialExpressionConstant3Vector_2, output=""}
   ]
   properties {}
  }
  MaterialExpressionDotProduct_9 {
   class=MaterialExpressionDotProduct
   inputs [
    A {from=MaterialExpressionNamedRerouteUsage_18, output=""}
    B {from=MaterialExpressionNamedRerouteUsage_18, output=""}
   ]
   properties {}
  }
  MaterialExpressionDynamicParameter_1 {
   class=MaterialExpressionDynamicParameter
   inputs []
   properties {parameterIndex=1, paramNames [General Brightness, Lightness Shift, RotatorTime, Mask Add OP]}
  }
  MaterialExpressionIf_0 {
   class=MaterialExpressionIf
   inputs [
    A {from=MaterialExpressionNamedRerouteDeclaration_35, output=""}
    B {from=MaterialExpressionConstant_1, output=""}
    "A > B" {from=MaterialExpressionMultiply_26, output=""}
    "A == B" {from=MaterialExpressionReroute_39, output=""}
    "A < B" {from=MaterialExpressionReroute_39, output=""}
   ]
   properties {constB=0.5}
  }
  MaterialExpressionIf_1 {
   class=MaterialExpressionIf
   inputs [
    A {from=MaterialExpressionNamedRerouteUsage_57, output=""}
    B {from=MaterialExpressionConstant_0, output=""}
    "A > B" {from=MaterialExpressionNamedRerouteUsage_58, output=""}
    "A == B" {from=MaterialExpressionReroute_5, output=""}
    "A < B" {from=MaterialExpressionReroute_5, output=""}
   ]
   properties {constB=0.5}
  }
  MaterialExpressionIf_2 {
   class=MaterialExpressionIf
   inputs [
    A {from=MaterialExpressionScalarParameter_13, output=""}
    B {from="", output=""}
    "A > B" {from=MaterialExpressionMaterialFunctionCall_6, output=Result}
    "A == B" {from=MaterialExpressionReroute_48, output=""}
    "A < B" {from=MaterialExpressionReroute_48, output=""}
   ]
   properties {constB=0.5}
  }
  MaterialExpressionLinearInterpolate_0 {
   class=MaterialExpressionLinearInterpolate
   inputs [
    A {from=MaterialExpressionVectorParameter_0, output=RGB}
    B {from=MaterialExpressionMultiply_1}
    Alpha {from=MaterialExpressionMaterialFunctionCall_8, output=Result}
   ]
  }
  MaterialExpressionLinearInterpolate_2 {
   class=MaterialExpressionLinearInterpolate
   inputs [
    A {from=MaterialExpressionConstantClamp_35}
    B {from=MaterialExpressionAdd_31}
    Alpha {from=MaterialExpressionScalarParameter_19}
   ]
  }
  MaterialExpressionLinearInterpolate_4 {
   class=MaterialExpressionLinearInterpolate
   inputs [
    A {from=MaterialExpressionNamedRerouteUsage_4}
    B {from=MaterialExpressionConstantClamp_29}
    Alpha {from=MaterialExpressionConstantClamp_29}
   ]
  }
  MaterialExpressionLinearInterpolate_6 {
   class=MaterialExpressionLinearInterpolate
   inputs [
    A {from=""}
    B {from=MaterialExpressionConstantBiasScale_18}
    Alpha {from=MaterialExpressionScalarParameter_3}
   ]
   properties {constB=0}
  }
  MaterialExpressionLinearInterpolate_7 {
   class=MaterialExpressionLinearInterpolate
   inputs [
    A {from=MaterialExpressionTextureSampleParameter2D_2, output=RGB}
    B {from=MaterialExpressionTextureSampleParameter2D_3, output=RGB}
    Alpha {from=""}
   ]
  }
  MaterialExpressionMaterialFunctionCall_0 {
   class=MaterialExpressionMaterialFunctionCall
   inputs [
    Shift {from=MaterialExpressionConstantClamp_58}
    Color {from=MaterialExpressionNamedRerouteUsage_43}
   ]
   properties {materialFunction=/DealershipSharedLibrary/FX/Materials/MaterialFunctions/MF_VCVFX_LightnessShift}
  }
  MaterialExpressionMaterialFunctionCall_1 {
   class=MaterialExpressionMaterialFunctionCall
   inputs [Emissive {from=MaterialExpressionReroute_37}]
   properties {materialFunction=&MF_FX_EmissiveAdjustments}
  }
  MaterialExpressionMaterialFunctionCall_11 {
   class=MaterialExpressionMaterialFunctionCall
   inputs [
    Time {from=""}
    Speed {from=MaterialExpressionScalarParameter_16}
   ]
   properties {materialFunction=&TimeWithSpeedVariable}
  }
  MaterialExpressionMaterialFunctionCall_2 {
   class=MaterialExpressionMaterialFunctionCall
   inputs [
    Color {from=MaterialExpressionMaterialFunctionCall_4, output=Result}
    Shift {from=MaterialExpressionNamedRerouteUsage_3}
   ]
   properties {materialFunction=/DealershipSharedLibrary/FX/Materials/MaterialFunctions/MF_MonochromaticShift}
  }
  MaterialExpressionMaterialFunctionCall_3 {
   class=MaterialExpressionMaterialFunctionCall
   inputs [
    Opacity In {from=MaterialExpressionReroute_30}
    DepthFade Distance {from=MaterialExpressionScalarParameter_12}
   ]
   properties {materialFunction=&MF_VCVFX_DepthFade}
  }
  MaterialExpressionMaterialFunctionCall_4 {
   class=MaterialExpressionMaterialFunctionCall
   inputs [
    Shift {from=MaterialExpressionNamedRerouteUsage_2}
    Color {from=MaterialExpressionNamedRerouteUsage_41}
   ]
   properties {materialFunction=&MF_VCVFX_LightnessShift}
  }
  MaterialExpressionMaterialFunctionCall_5 {
   class=MaterialExpressionMaterialFunctionCall
   inputs [
    Vector or UVs {from=MaterialExpressionAppendVector_14}
    Swizzle Coordinate Output {from=""}
   ]
   properties {materialFunction=&VectorToRadialValue}
  }
  MaterialExpressionMaterialFunctionCall_6 {
   class=MaterialExpressionMaterialFunctionCall
   inputs [
    Alpha Threshold {from=MaterialExpressionReroute_46}
    Random {from=""}
   ]
   properties {materialFunction=&DitherTemporalAA}
  }
  MaterialExpressionMaterialFunctionCall_7 {
   class=MaterialExpressionMaterialFunctionCall
   inputs [
    Shift {from=MaterialExpressionNamedRerouteUsage_38}
    Color {from=MaterialExpressionMaterialFunctionCall_0, output=Result}
   ]
   properties {materialFunction=&MF_VCVFX_LightnessShift}
  }
  MaterialExpressionMaterialFunctionCall_8 {
   class=MaterialExpressionMaterialFunctionCall
   inputs [
    In {from=MaterialExpressionSmoothStep_0}
    Contrast {from=MaterialExpressionScalarParameter_21}
   ]
   properties {materialFunction=&CheapContrast}
  }
  MaterialExpressionMeshEmitterDynamicParameter_2 {class=MaterialExpressionDynamicParameter, inputs [], properties {paramNames [Brightness,Shift,Tiling,Speed]}}
  MaterialExpressionMeshEmitterVertexColor_2 {class=MaterialExpressionParticleColor, inputs [], properties {}}
  MaterialExpressionMultiply_0 {
   class=MaterialExpressionMultiply
   inputs [
    A {from=MaterialExpressionSaturate_3}
    B {from=MaterialExpressionNamedRerouteUsage_59}
   ]
  }
  MaterialExpressionMultiply_1 {
   class=MaterialExpressionMultiply
   inputs [
    A {from=MaterialExpressionNamedRerouteUsage_0}
    B {from=MaterialExpressionNamedRerouteUsage_6}
   ]
  }
  MaterialExpressionMultiply_100 {
   class=MaterialExpressionMultiply
   inputs [
    A {from=MaterialExpressionReroute_26}
    B {from=MaterialExpressionNamedRerouteUsage_33}
   ]
  }
  MaterialExpressionMultiply_102 {
   class=MaterialExpressionMultiply
   inputs [
    A {from=MaterialExpressionConstantClamp_44}
    B {from=""}
   ]
   properties {constB=2}
  }
  MaterialExpressionMultiply_103 {
   class=MaterialExpressionMultiply
   inputs [
    A {from=MaterialExpressionConstantClamp_49}
    B {from=MaterialExpressionConstantClamp_50}
   ]
  }
  MaterialExpressionMultiply_104 {
   class=MaterialExpressionMultiply
   inputs [
    A {from=MaterialExpressionConstantClamp_55}
    B {from=MaterialExpressionConstantClamp_50}
   ]
  }
  MaterialExpressionMultiply_105 {
   class=MaterialExpressionMultiply
   inputs [
    A {from=MaterialExpressionConstantClamp_54}
    B {from=""}
   ]
   properties {constB=2}
  }
  MaterialExpressionMultiply_106 {
   class=MaterialExpressionMultiply
   inputs [
    A {from=MaterialExpressionMultiply_95}
    B {from=MaterialExpressionConstant_38}
   ]
  }
  MaterialExpressionMultiply_107 {
   class=MaterialExpressionMultiply
   inputs [
    A {from=MaterialExpressionNamedRerouteUsage_44}
    B {from=MaterialExpressionConstantClamp_60}
   ]
  }
  MaterialExpressionMultiply_108 {
   class=MaterialExpressionMultiply
   inputs [
    A {from=MaterialExpressionConstantClamp_59}
    B {from=MaterialExpressionMultiply_107}
   ]
  }
  MaterialExpressionMultiply_2 {
   class=MaterialExpressionMultiply
   inputs [
    A {from=MaterialExpressionMultiply_72}
    B {from=MaterialExpressionScalarParameter_1}
   ]
   properties {constB=20}
  }
  MaterialExpressionMultiply_20 {
   class=MaterialExpressionMultiply
   inputs [
    A {from=MaterialExpressionMaterialFunctionCall_7, output=Result}
    B {from=MaterialExpressionNamedRerouteUsage_45}
   ]
  }
  MaterialExpressionMultiply_22 {
   class=MaterialExpressionMultiply
   inputs [
    A {from=MaterialExpressionAdd_22}
    B {from=MaterialExpressionNamedRerouteUsage_32}
   ]
  }
  MaterialExpressionMultiply_23 {
   class=MaterialExpressionMultiply
   inputs [
    A {from=MaterialExpressionScalarParameter_6}
    B {from=MaterialExpressionNamedRerouteUsage_31}
   ]
  }
  MaterialExpressionMultiply_24 {
   class=MaterialExpressionMultiply
   inputs [
    A {from=MaterialExpressionMultiply_25}
    B {from=MaterialExpressionConstant_28}
   ]
  }
  MaterialExpressionMultiply_25 {
   class=MaterialExpressionMultiply
   inputs [
    A {from=MaterialExpressionMultiply_84}
    B {from=MaterialExpressionComponentMask_5}
   ]
  }
  MaterialExpressionMultiply_26 {
   class=MaterialExpressionMultiply
   inputs [
    A {from=MaterialExpressionMultiply_20}
    B {from=MaterialExpressionNamedRerouteUsage_39}
   ]
  }
  MaterialExpressionMultiply_3 {
   class=MaterialExpressionMultiply
   inputs [
    A {from=MaterialExpressionMultiply_68}
    B {from=MaterialExpressionScalarParameter_14}
   ]
  }
  MaterialExpressionMultiply_30 {
   class=MaterialExpressionMultiply
   inputs [
    A {from=MaterialExpressionAppendVector_13}
    B {from=MaterialExpressionAdd_34}
   ]
  }
  MaterialExpressionMultiply_58 {
   class=MaterialExpressionMultiply
   inputs [
    A {from=MaterialExpressionNamedRerouteUsage_25}
    B {from=MaterialExpressionConstantClamp_31}
   ]
  }
  MaterialExpressionMultiply_62 {
   class=MaterialExpressionMultiply
   inputs [
    A {from=MaterialExpressionMultiply_86}
    B {from=MaterialExpressionScalarParameter_9}
   ]
  }
  MaterialExpressionMultiply_63 {
   class=MaterialExpressionMultiply
   inputs [
    A {from=MaterialExpressionLinearInterpolate_4}
    B {from=MaterialExpressionConstantClamp_26}
   ]
  }
  MaterialExpressionMultiply_64 {
   class=MaterialExpressionMultiply
   inputs [
    A {from=MaterialExpressionMultiply_58}
    B {from=MaterialExpressionScalarParameter_5}
   ]
  }
  MaterialExpressionMultiply_65 {
   class=MaterialExpressionMultiply
   inputs [
    A {from=MaterialExpressionMultiply_66}
    B {from=MaterialExpressionScalarParameter_4}
   ]
  }
  MaterialExpressionMultiply_66 {
   class=MaterialExpressionMultiply
   inputs [
    A {from=MaterialExpressionConstantClamp_32}
    B {from=MaterialExpressionConstantClamp_32}
   ]
  }
  MaterialExpressionMultiply_68 {
   class=MaterialExpressionMultiply
   inputs [
    A {from=MaterialExpressionConvert_1}
    B {from=MaterialExpressionScalarParameter_15}
   ]
  }
  MaterialExpressionMultiply_69 {
   class=MaterialExpressionMultiply
   inputs [
    A {from=MaterialExpressionNamedRerouteUsage_26}
    B {from=MaterialExpressionNamedRerouteUsage_24}
   ]
  }
  MaterialExpressionMultiply_70 {
   class=MaterialExpressionMultiply
   inputs [
    A {from=MaterialExpressionComponentMask_4}
    B {from=MaterialExpressionNamedRerouteUsage_9}
   ]
  }
  MaterialExpressionMultiply_71 {
   class=MaterialExpressionMultiply
   inputs [
    A {from=MaterialExpressionConstantClamp_36}
    B {from=MaterialExpressionConstantClamp_36}
   ]
  }
  MaterialExpressionMultiply_72 {
   class=MaterialExpressionMultiply
   inputs [
    A {from=MaterialExpressionNamedRerouteUsage_65}
    B {from=MaterialExpressionMultiply_85}
   ]
  }
  MaterialExpressionMultiply_74 {
   class=MaterialExpressionMultiply
   inputs [
    A {from=MaterialExpressionNamedRerouteUsage_20}
    B {from=MaterialExpressionNamedRerouteUsage_14}
   ]
  }
  MaterialExpressionMultiply_75 {
   class=MaterialExpressionMultiply
   inputs [
    A {from=MaterialExpressionNamedRerouteUsage_22}
    B {from=MaterialExpressionNamedRerouteUsage_15}
   ]
  }
  MaterialExpressionMultiply_76 {
   class=MaterialExpressionMultiply
   inputs [
    A {from=MaterialExpressionConstantBiasScale_9}
    B {from=MaterialExpressionNamedRerouteUsage_7}
   ]
  }
  MaterialExpressionMultiply_77 {
   class=MaterialExpressionMultiply
   inputs [
    A {from=MaterialExpressionNamedRerouteUsage_13}
    B {from=MaterialExpressionNamedRerouteUsage_13}
   ]
  }
  MaterialExpressionMultiply_78 {
   class=MaterialExpressionMultiply
   inputs [
    A {from=MaterialExpressionNamedRerouteUsage_12}
    B {from=MaterialExpressionNamedRerouteUsage_12}
   ]
  }
  MaterialExpressionMultiply_79 {
   class=MaterialExpressionMultiply
   inputs [
    A {from=MaterialExpressionNamedRerouteUsage_19}
    B {from=MaterialExpressionNamedRerouteUsage_17}
   ]
  }
  MaterialExpressionMultiply_80 {
   class=MaterialExpressionMultiply
   inputs [
    A {from=MaterialExpressionNamedRerouteUsage_21}
    B {from=MaterialExpressionNamedRerouteUsage_16}
   ]
  }
  MaterialExpressionMultiply_81 {
   class=MaterialExpressionMultiply
   inputs [
    A {from=MaterialExpressionTextureSampleParameter2D_4, output=R}
    B {from=MaterialExpressionTextureSampleParameter2D_4, output=R}
   ]
  }
  MaterialExpressionMultiply_82 {
   class=MaterialExpressionMultiply
   inputs [
    A {from=MaterialExpressionTextureSampleParameter2D_4, output=G}
    B {from=MaterialExpressionTextureSampleParameter2D_4, output=G}
   ]
  }
  MaterialExpressionMultiply_83 {
   class=MaterialExpressionMultiply
   inputs [
    A {from=MaterialExpressionSubtract_10}
    B {from=MaterialExpressionSubtract_10}
   ]
  }
  MaterialExpressionMultiply_84 {
   class=MaterialExpressionMultiply
   inputs [
    A {from=MaterialExpressionConstantClamp_47}
    B {from=MaterialExpressionMultiply_87}
   ]
  }
  MaterialExpressionMultiply_85 {
   class=MaterialExpressionMultiply
   inputs [
    A {from=MaterialExpressionNamedRerouteDeclaration_3}
    B {from=MaterialExpressionNamedRerouteDeclaration_3}
   ]
  }
  MaterialExpressionMultiply_86 {
   class=MaterialExpressionMultiply
   inputs [
    A {from=MaterialExpressionMaterialFunctionCall_2, output=Result}
    B {from=MaterialExpressionNamedRerouteUsage_10}
   ]
  }
  MaterialExpressionMultiply_87 {
   class=MaterialExpressionMultiply
   inputs [
    A {from=MaterialExpressionConstantClamp_48}
    B {from=MaterialExpressionConstantClamp_48}
   ]
  }
  MaterialExpressionMultiply_88 {
   class=MaterialExpressionMultiply
   inputs [
    A {from=MaterialExpressionSubtract_7}
    B {from=MaterialExpressionScalarParameter_10}
   ]
  }
  MaterialExpressionMultiply_89 {
   class=MaterialExpressionMultiply
   inputs [
    A {from=MaterialExpressionConstantClamp_57}
    B {from=MaterialExpressionConstantClamp_50}
   ]
  }
  MaterialExpressionMultiply_90 {
   class=MaterialExpressionMultiply
   inputs [
    A {from=MaterialExpressionConstantClamp_56}
    B {from=""}
   ]
   properties {constB=5}
  }
  MaterialExpressionMultiply_91 {
   class=MaterialExpressionMultiply
   inputs [
    A {from=MaterialExpressionConstantClamp_53}
    B {from=MaterialExpressionConstant_37}
   ]
  }
  MaterialExpressionMultiply_92 {
   class=MaterialExpressionMultiply
   inputs [
    A {from=MaterialExpressionNamedRerouteUsage_54}
    B {from=MaterialExpressionNamedRerouteUsage_47}
   ]
  }
  MaterialExpressionMultiply_93 {
   class=MaterialExpressionMultiply
   inputs [
    A {from=MaterialExpressionNamedRerouteUsage_50}
    B {from=MaterialExpressionNamedRerouteUsage_48}
   ]
  }
  MaterialExpressionMultiply_94 {
   class=MaterialExpressionMultiply
   inputs [
    A {from=MaterialExpressionReroute_27}
    B {from=MaterialExpressionConvert_2}
   ]
  }
  MaterialExpressionMultiply_95 {
   class=MaterialExpressionMultiply
   inputs [
    A {from=MaterialExpressionConstantBiasScale_34}
    B {from=MaterialExpressionScalarParameter_11}
   ]
  }
  MaterialExpressionMultiply_96 {
   class=MaterialExpressionMultiply
   inputs [
    A {from=MaterialExpressionMultiply_22}
    B {from=MaterialExpressionConstant_33}
   ]
  }
  MaterialExpressionMultiply_97 {
   class=MaterialExpressionMultiply
   inputs [
    A {from=MaterialExpressionReroute_28}
    B {from=MaterialExpressionReroute_26}
   ]
  }
  MaterialExpressionMultiply_98 {
   class=MaterialExpressionMultiply
   inputs [
    A {from=MaterialExpressionNamedRerouteUsage_52}
    B {from=MaterialExpressionNamedRerouteUsage_46}
   ]
  }
  MaterialExpressionMultiply_99 {
   class=MaterialExpressionMultiply
   inputs [
    A {from=MaterialExpressionNamedRerouteUsage_53}
    B {from=MaterialExpressionNamedRerouteUsage_49}
   ]
  }
  MaterialExpressionNamedRerouteDeclaration_0 {
   class=MaterialExpressionNamedRerouteDeclaration
   inputs [None {from=MaterialExpressionMultiply_58}]
   properties {name=FinalMask_Color, nodeColor=<rgba 1 0.2812 0 1>}
  }
  MaterialExpressionNamedRerouteDeclaration_1 {
   class=MaterialExpressionNamedRerouteDeclaration
   inputs [None {from=MaterialExpressionMeshEmitterDynamicParameter_2, output=Brightness}]
   properties {name=DYN Brightness, nodeColor=<rgba 1 0 0 1>}
  }
  MaterialExpressionNamedRerouteDeclaration_10 {
   class=MaterialExpressionNamedRerouteDeclaration
   inputs [None {from=MaterialExpressionConvert_0}]
   properties {name=TexCoords_MaskG, nodeColor=<rgba 0 1 0.0859 1>}
  }
  MaterialExpressionNamedRerouteDeclaration_11 {
   class=MaterialExpressionNamedRerouteDeclaration
   inputs [None {from=MaterialExpressionConvert_0}]
   properties {name=TexCoords_MaskR, nodeColor=<rgba 1 0 0.2344 1>}
  }
  MaterialExpressionNamedRerouteDeclaration_12 {
   class=MaterialExpressionNamedRerouteDeclaration
   inputs [None {from=MaterialExpressionCosine_5}]
   properties {name=RR_Cosine, nodeColor=<rgba 1 0 0.3984 1>}
  }
  MaterialExpressionNamedRerouteDeclaration_13 {
   class=MaterialExpressionNamedRerouteDeclaration
   inputs [None {from=MaterialExpressionSine_5}]
   properties {name=RR_Sine, nodeColor=<rgba 0 0.7187 1 1>}
  }
  MaterialExpressionNamedRerouteDeclaration_15 {
   class=MaterialExpressionNamedRerouteDeclaration
   inputs [None {from=MaterialExpressionLinearInterpolate_2}]
   properties {name=Detail PatternAnd Core, nodeColor=<rgba 0.3828 0 1 1>}
  }
  MaterialExpressionNamedRerouteDeclaration_16 {
   class=MaterialExpressionNamedRerouteDeclaration
   inputs [None {from=MaterialExpressionConstantClamp_30}]
   properties {name=SphereMask, nodeColor=<rgba 1 0.9375 0 1>}
  }
  MaterialExpressionNamedRerouteDeclaration_17 {
   class=MaterialExpressionNamedRerouteDeclaration
   inputs [None {from=MaterialExpressionMultiply_64}]
   properties {name=FinalMask, nodeColor=<rgba 1 0.2812 0 1>}
  }
  MaterialExpressionNamedRerouteDeclaration_18 {
   class=MaterialExpressionNamedRerouteDeclaration
   inputs [None {from=MaterialExpressionLinearInterpolate_6}]
   properties {name=WPO, nodeColor=<rgba 0.4297 0 1 1>}
  }
  MaterialExpressionNamedRerouteDeclaration_19 {
   class=MaterialExpressionNamedRerouteDeclaration
   inputs [None {from=MaterialExpressionDotProduct_9}]
   properties {name=TexCoordMaskDot, nodeColor=<rgba 0.1484 1 0 1>}
  }
  MaterialExpressionNamedRerouteDeclaration_2 {
   class=MaterialExpressionNamedRerouteDeclaration
   inputs [None {from=MaterialExpressionMeshEmitterDynamicParameter_2, output=Shift}]
   properties {name=DYN Shift Color, nodeColor=<rgba 1 0 0 1>}
  }
  MaterialExpressionNamedRerouteDeclaration_20 {
   class=MaterialExpressionNamedRerouteDeclaration
   inputs [None {from=MaterialExpressionDynamicParameter_1, output=RotatorTime}]
   properties {name=DYN RotatorTime, nodeColor=<rgba 1 0 0 1>}
  }
  MaterialExpressionNamedRerouteDeclaration_21 {
   class=MaterialExpressionNamedRerouteDeclaration
   inputs [None {from=MaterialExpressionOneMinus_3}]
   properties {name=SphereMask1-, nodeColor=<rgba 0.0781 0 1 1>}
  }
  MaterialExpressionNamedRerouteDeclaration_22 {
   class=MaterialExpressionNamedRerouteDeclaration
   inputs [None {from=MaterialExpressionDynamicParameter_1, output=Mask Add OP}]
   properties {name=DYN Mask Add Operation, nodeColor=<rgba 1 0 0 1>}
  }
  MaterialExpressionNamedRerouteDeclaration_23 {
   class=MaterialExpressionNamedRerouteDeclaration
   inputs [None {from=MaterialExpressionDynamicParameter_1, output=Lightness Shift}]
   properties {name=DYN Lightness Shift, nodeColor=<rgba 1 0 0 1>}
  }
  MaterialExpressionNamedRerouteDeclaration_24 {
   class=MaterialExpressionNamedRerouteDeclaration
   inputs [None {from=MaterialExpressionDynamicParameter_1, output=General Brightness}]
   properties {name=DYN General Brightness, nodeColor=<rgba 1 0 0 1>}
  }
  MaterialExpressionNamedRerouteDeclaration_25 {
   class=MaterialExpressionNamedRerouteDeclaration
   inputs [None {from=MaterialExpressionConstantClamp_46}]
   properties {name=Swirl Sphere Mask, nodeColor=<rgba 0 1 0.7891 1>}
  }
  MaterialExpressionNamedRerouteDeclaration_26 {
   class=MaterialExpressionNamedRerouteDeclaration
   inputs [None {from=MaterialExpressionMeshEmitterVertexColor_2, output=RGB}]
   properties {name=Particle Color RGB, nodeColor=<rgba 1 0.1048 0 1>}
  }
  MaterialExpressionNamedRerouteDeclaration_27 {
   class=MaterialExpressionNamedRerouteDeclaration
   inputs [None {from=MaterialExpressionMeshEmitterVertexColor_2, output=A}]
   properties {name=Particle Color Alpha, nodeColor=<rgba 1 0.0779 0.0003 1>}
  }
  MaterialExpressionNamedRerouteDeclaration_28 {
   class=MaterialExpressionNamedRerouteDeclaration
   inputs [None {from=MaterialExpressionDotProduct_11}]
   properties {name=Swirld Mask To Color, nodeColor=<rgba 0.1719 1 0 1>}
  }
  MaterialExpressionNamedRerouteDeclaration_29 {
   class=MaterialExpressionNamedRerouteDeclaration
   inputs [None {from=MaterialExpressionMultiply_94}]
   properties {name=Swirl Top, nodeColor=<rgba 0.2188 0 1 1>}
  }
  MaterialExpressionNamedRerouteDeclaration_3 {
   class=MaterialExpressionNamedRerouteDeclaration
   inputs [None {from=MaterialExpressionConstantClamp_28}]
   properties {name=Mask_02, nodeColor=<rgba 0 0.7891 1 1>}
  }
  MaterialExpressionNamedRerouteDeclaration_30 {
   class=MaterialExpressionNamedRerouteDeclaration
   inputs [None {from=MaterialExpressionMultiply_97}]
   properties {name=Swirl right, nodeColor=<rgba 0.1016 1 0 1>}
  }
  MaterialExpressionNamedRerouteDeclaration_31 {
   class=MaterialExpressionNamedRerouteDeclaration
   inputs [None {from=MaterialExpressionCosine_6}]
   properties {name=Swirl Cosine, nodeColor=<rgba 1 0 0.4219 1>}
  }
  MaterialExpressionNamedRerouteDeclaration_32 {
   class=MaterialExpressionNamedRerouteDeclaration
   inputs [None {from=MaterialExpressionConstantClamp_52}]
   properties {name=Mask Bias, nodeColor=<rgba 0 0.7422 1 1>}
  }
  MaterialExpressionNamedRerouteDeclaration_33 {
   class=MaterialExpressionNamedRerouteDeclaration
   inputs [None {from=MaterialExpressionSine_1}]
   properties {name=Swirl Sine, nodeColor=<rgba 1 0.9375 0 1>}
  }
  MaterialExpressionNamedRerouteDeclaration_34 {
   class=MaterialExpressionNamedRerouteDeclaration
   inputs [None {from=MaterialExpressionDotProduct_14}]
   properties {name=Swirl Dot Sine op, nodeColor=<rgba 0.6172 0 1 1>}
  }
  MaterialExpressionNamedRerouteDeclaration_35 {
   class=MaterialExpressionNamedRerouteDeclaration
   inputs [None {from=MaterialExpressionScalarParameter_0}]
   properties {name=UseSwirl?, nodeColor=<rgba 1 0 0.1172 1>}
  }
  MaterialExpressionNamedRerouteDeclaration_36 {
   class=MaterialExpressionNamedRerouteDeclaration
   inputs [None {from=MaterialExpressionSaturate_1}]
   properties {name=Swirl alpha, nodeColor=<rgba 1 0 0.2109 1>}
  }
  MaterialExpressionNamedRerouteDeclaration_37 {
   class=MaterialExpressionNamedRerouteDeclaration
   inputs [None {from=MaterialExpressionConstantBiasScale_0}]
   properties {name=Mask, nodeColor=<rgba 1 0 0.4922 1>}
  }
  MaterialExpressionNamedRerouteDeclaration_4 {
   class=MaterialExpressionNamedRerouteDeclaration
   inputs [None {from=MaterialExpressionMultiply_71}]
   properties {name=Mask_03, nodeColor=<rgba 1 0.8906 0 1>}
  }
  MaterialExpressionNamedRerouteDeclaration_41 {
   class=MaterialExpressionNamedRerouteDeclaration
   inputs [None {from=MaterialExpressionDotProduct_0}]
   properties {name=sphere, nodeColor=<rgba 1 0 0.7358 1>}
  }
  MaterialExpressionNamedRerouteDeclaration_5 {
   class=MaterialExpressionNamedRerouteDeclaration
   inputs [None {from=MaterialExpressionMultiply_83}]
   properties {name=Mask_04, nodeColor=<rgba 0.5703 0 1 1>}
  }
  MaterialExpressionNamedRerouteDeclaration_6 {
   class=MaterialExpressionNamedRerouteDeclaration
   inputs [None {from=MaterialExpressionMeshEmitterDynamicParameter_2, output=Tiling}]
   properties {name=DYN Tiling Portal TEX, nodeColor=<rgba 1 0 0 1>}
  }
  MaterialExpressionNamedRerouteDeclaration_7 {
   class=MaterialExpressionNamedRerouteDeclaration
   inputs [None {from=MaterialExpressionMeshEmitterDynamicParameter_2, output=Speed}]
   properties {name=DYN Speed Portal TEX, nodeColor=<rgba 1 0 0 1>}
  }
  MaterialExpressionNamedRerouteDeclaration_8 {
   class=MaterialExpressionNamedRerouteDeclaration
   inputs [None {from=MaterialExpressionAdd_33}]
   properties {name=Mask_Result, nodeColor=<rgba 0 0.3906 1 1>}
  }
  MaterialExpressionNamedRerouteDeclaration_9 {
   class=MaterialExpressionNamedRerouteDeclaration
   inputs [None {from=MaterialExpressionAdd_12}]
   properties {name=TexCoord Masks, nodeColor=<rgba 0.4062 0 1 1>}
  }
  MaterialExpressionNamedRerouteUsage_0 {class=MaterialExpressionNamedRerouteUsage}
  MaterialExpressionNamedRerouteUsage_1 {class=MaterialExpressionNamedRerouteUsage}
  MaterialExpressionNamedRerouteUsage_10 {class=MaterialExpressionNamedRerouteUsage}
  MaterialExpressionNamedRerouteUsage_11 {class=MaterialExpressionNamedRerouteUsage}
  MaterialExpressionNamedRerouteUsage_12 {class=MaterialExpressionNamedRerouteUsage}
  MaterialExpressionNamedRerouteUsage_13 {class=MaterialExpressionNamedRerouteUsage}
  MaterialExpressionNamedRerouteUsage_14 {class=MaterialExpressionNamedRerouteUsage}
  MaterialExpressionNamedRerouteUsage_15 {class=MaterialExpressionNamedRerouteUsage}
  MaterialExpressionNamedRerouteUsage_16 {class=MaterialExpressionNamedRerouteUsage}
  MaterialExpressionNamedRerouteUsage_17 {class=MaterialExpressionNamedRerouteUsage}
  MaterialExpressionNamedRerouteUsage_18 {class=MaterialExpressionNamedRerouteUsage}
  MaterialExpressionNamedRerouteUsage_19 {class=MaterialExpressionNamedRerouteUsage}
  MaterialExpressionNamedRerouteUsage_2 {class=MaterialExpressionNamedRerouteUsage}
  MaterialExpressionNamedRerouteUsage_20 {class=MaterialExpressionNamedRerouteUsage}
  MaterialExpressionNamedRerouteUsage_21 {class=MaterialExpressionNamedRerouteUsage}
  MaterialExpressionNamedRerouteUsage_22 {class=MaterialExpressionNamedRerouteUsage}
  MaterialExpressionNamedRerouteUsage_24 {class=MaterialExpressionNamedRerouteUsage}
  MaterialExpressionNamedRerouteUsage_25 {class=MaterialExpressionNamedRerouteUsage}
  MaterialExpressionNamedRerouteUsage_26 {class=MaterialExpressionNamedRerouteUsage}
  MaterialExpressionNamedRerouteUsage_28 {class=MaterialExpressionNamedRerouteUsage}
  MaterialExpressionNamedRerouteUsage_29 {class=MaterialExpressionNamedRerouteUsage}
  MaterialExpressionNamedRerouteUsage_3 {class=MaterialExpressionNamedRerouteUsage}
  MaterialExpressionNamedRerouteUsage_30 {class=MaterialExpressionNamedRerouteUsage}
  MaterialExpressionNamedRerouteUsage_31 {class=MaterialExpressionNamedRerouteUsage}
  MaterialExpressionNamedRerouteUsage_32 {class=MaterialExpressionNamedRerouteUsage}
  MaterialExpressionNamedRerouteUsage_33 {class=MaterialExpressionNamedRerouteUsage}
  MaterialExpressionNamedRerouteUsage_34 {class=MaterialExpressionNamedRerouteUsage}
  MaterialExpressionNamedRerouteUsage_35 {class=MaterialExpressionNamedRerouteUsage}
  MaterialExpressionNamedRerouteUsage_36 {class=MaterialExpressionNamedRerouteUsage}
  MaterialExpressionNamedRerouteUsage_37 {class=MaterialExpressionNamedRerouteUsage}
  MaterialExpressionNamedRerouteUsage_38 {class=MaterialExpressionNamedRerouteUsage}
  MaterialExpressionNamedRerouteUsage_39 {class=MaterialExpressionNamedRerouteUsage, properties {desc=this value needs to be the half from UE3}}
  MaterialExpressionNamedRerouteUsage_4 {class=MaterialExpressionNamedRerouteUsage}
  MaterialExpressionNamedRerouteUsage_40 {class=MaterialExpressionNamedRerouteUsage}
  MaterialExpressionNamedRerouteUsage_41 {class=MaterialExpressionNamedRerouteUsage}
  MaterialExpressionNamedRerouteUsage_42 {class=MaterialExpressionNamedRerouteUsage}
  MaterialExpressionNamedRerouteUsage_43 {class=MaterialExpressionNamedRerouteUsage}
  MaterialExpressionNamedRerouteUsage_44 {class=MaterialExpressionNamedRerouteUsage}
  MaterialExpressionNamedRerouteUsage_45 {class=MaterialExpressionNamedRerouteUsage}
  MaterialExpressionNamedRerouteUsage_46 {class=MaterialExpressionNamedRerouteUsage}
  MaterialExpressionNamedRerouteUsage_47 {class=MaterialExpressionNamedRerouteUsage}
  MaterialExpressionNamedRerouteUsage_48 {class=MaterialExpressionNamedRerouteUsage}
  MaterialExpressionNamedRerouteUsage_49 {class=MaterialExpressionNamedRerouteUsage}
  MaterialExpressionNamedRerouteUsage_5 {class=MaterialExpressionNamedRerouteUsage}
  MaterialExpressionNamedRerouteUsage_50 {class=MaterialExpressionNamedRerouteUsage}
  MaterialExpressionNamedRerouteUsage_51 {class=MaterialExpressionNamedRerouteUsage}
  MaterialExpressionNamedRerouteUsage_52 {class=MaterialExpressionNamedRerouteUsage}
  MaterialExpressionNamedRerouteUsage_53 {class=MaterialExpressionNamedRerouteUsage}
  MaterialExpressionNamedRerouteUsage_54 {class=MaterialExpressionNamedRerouteUsage}
  MaterialExpressionNamedRerouteUsage_55 {class=MaterialExpressionNamedRerouteUsage}
  MaterialExpressionNamedRerouteUsage_56 {class=MaterialExpressionNamedRerouteUsage}
  MaterialExpressionNamedRerouteUsage_57 {class=MaterialExpressionNamedRerouteUsage}
  MaterialExpressionNamedRerouteUsage_58 {class=MaterialExpressionNamedRerouteUsage}
  MaterialExpressionNamedRerouteUsage_59 {class=MaterialExpressionNamedRerouteUsage}
  MaterialExpressionNamedRerouteUsage_6 {class=MaterialExpressionNamedRerouteUsage}
  MaterialExpressionNamedRerouteUsage_60 {class=MaterialExpressionNamedRerouteUsage}
  MaterialExpressionNamedRerouteUsage_65 {class=MaterialExpressionNamedRerouteUsage}
  MaterialExpressionNamedRerouteUsage_66 {class=MaterialExpressionNamedRerouteUsage}
  MaterialExpressionNamedRerouteUsage_7 {class=MaterialExpressionNamedRerouteUsage}
  MaterialExpressionNamedRerouteUsage_8 {class=MaterialExpressionNamedRerouteUsage}
  MaterialExpressionNamedRerouteUsage_9 {class=MaterialExpressionNamedRerouteUsage}
  MaterialExpressionOneMinus_2 {class=MaterialExpressionOneMinus, inputs [None {from=MaterialExpressionNamedRerouteUsage_30}]}
  MaterialExpressionOneMinus_3 {class=MaterialExpressionOneMinus, inputs [None {from=MaterialExpressionConstantClamp_46}]}
  MaterialExpressionOneMinus_7 {class=MaterialExpressionOneMinus, inputs [None {from=MaterialExpressionNamedRerouteDeclaration_19}]}
  MaterialExpressionOneMinus_8 {class=MaterialExpressionOneMinus, inputs [None {from=MaterialExpressionConstantClamp_33}]}
  MaterialExpressionPanner_7 {
   class=MaterialExpressionPanner
   inputs [
    Coordinate {from=MaterialExpressionMultiply_30}
    Time {from=MaterialExpressionNamedRerouteUsage_8}
    Speed {from=""}
   ]
   properties {speedX=1.725, speedY=0.25}
  }
  MaterialExpressionPower_2 {
   class=MaterialExpressionPower
   inputs [
    Base {from=MaterialExpressionDotProduct_8}
    Exp {from=MaterialExpressionScalarParameter_18}
   ]
  }
  MaterialExpressionReroute_26 {class=MaterialExpressionReroute, inputs [None {from=MaterialExpressionScalarParameter_7}]}
  MaterialExpressionReroute_27 {class=MaterialExpressionReroute, inputs [None {from=MaterialExpressionScalarParameter_7}]}
  MaterialExpressionReroute_28 {class=MaterialExpressionReroute, inputs [None {from=MaterialExpressionConvert_2}]}
  MaterialExpressionReroute_30 {class=MaterialExpressionReroute, inputs [None {from=MaterialExpressionIf_1}]}
  MaterialExpressionReroute_37 {class=MaterialExpressionReroute, inputs [None {from=MaterialExpressionIf_0}]}
  MaterialExpressionReroute_38 {class=MaterialExpressionReroute, inputs [None {from=MaterialExpressionMultiply_62}]}
  MaterialExpressionReroute_39 {class=MaterialExpressionReroute, inputs [None {from=MaterialExpressionReroute_38}]}
  MaterialExpressionReroute_46 {class=MaterialExpressionReroute, inputs [None {from=MaterialExpressionIf_1}]}
  MaterialExpressionReroute_47 {class=MaterialExpressionReroute, inputs [None {from=MaterialExpressionReroute_46}]}
  MaterialExpressionReroute_48 {class=MaterialExpressionReroute, inputs [None {from=MaterialExpressionReroute_47}]}
  MaterialExpressionReroute_5 {class=MaterialExpressionReroute, inputs [None {from=MaterialExpressionConstantClamp_27}]}
  MaterialExpressionRotator_1 {
   class=MaterialExpressionRotator
   inputs [
    Coordinate {from=MaterialExpressionNamedRerouteUsage_60, output=""}
    Time {from="", output=""}
   ]
   properties {centerX=0, centerY=0, speed=-0.25}
  }
  MaterialExpressionRotator_6 {
   class=MaterialExpressionRotator
   inputs [
    Coordinate {from=MaterialExpressionTextureCoordinate_10, output=""}
    Time {from=MaterialExpressionMultiply_23, output=""}
   ]
   properties {speed=3.14}
  }
  MaterialExpressionRotator_7 {
   class=MaterialExpressionRotator
   inputs [
    Coordinate {from=MaterialExpressionTextureCoordinate_12, output=""}
    Time {from="", output=""}
   ]
   properties {speed=-0.075}
  }
  MaterialExpressionRotator_8 {
   class=MaterialExpressionRotator
   inputs [
    Coordinate {from=MaterialExpressionTextureCoordinate_13, output=""}
    Time {from="", output=""}
   ]
   properties {centerX=-0.5, speed=0.1}
  }
  MaterialExpressionSaturate_1 {class=MaterialExpressionSaturate, inputs [None {from=MaterialExpressionMultiply_108}]}
  MaterialExpressionSaturate_2 {class=MaterialExpressionSaturate, inputs [None {from=MaterialExpressionNamedRerouteDeclaration_41}]}
  MaterialExpressionSaturate_3 {class=MaterialExpressionSaturate, inputs [None {from=MaterialExpressionNamedRerouteDeclaration_37}]}
  MaterialExpressionScalarParameter_0 {
   class=MaterialExpressionScalarParameter
   properties {
    controlType=Enumeration
    enumeration=/DealershipSharedLibrary/FX/Materials/Enums/Enum_DLR_TrueFalse
    group=HypnoOrSwirl
    parameterName=Use Swirl? 0 False 1 True
   }
  }
  MaterialExpressionScalarParameter_1 {class=MaterialExpressionScalarParameter, properties {defaultValue=1, group=Group Hypno, parameterName=CenterMaskControl}}
  MaterialExpressionScalarParameter_10 {class=MaterialExpressionScalarParameter, properties {defaultValue=10, parameterName=Circle Brightness}}
  MaterialExpressionScalarParameter_11 {class=MaterialExpressionScalarParameter, properties {defaultValue=1.5, group=Group Swirl, parameterName=SwirlTile}}
  MaterialExpressionScalarParameter_12 {class=MaterialExpressionScalarParameter, properties {defaultValue=100, parameterName=DepthFade Distance}}
  MaterialExpressionScalarParameter_13 {
   class=MaterialExpressionScalarParameter
   properties {controlType=Enumeration, enumeration=&Enum_DLR_TrueFalse, parameterName=Use DitherTemporalAA?}
  }
  MaterialExpressionScalarParameter_14 {class=MaterialExpressionScalarParameter, properties {defaultValue=2, parameterName=amount}}
  MaterialExpressionScalarParameter_15 {class=MaterialExpressionScalarParameter, properties {defaultValue=1.5, group=WPO, parameterName=Intensity, sortPriority=3}}
  MaterialExpressionScalarParameter_16 {class=MaterialExpressionScalarParameter, properties {defaultValue=-1.5, group=WPO, parameterName=Speed, sortPriority=2}}
  MaterialExpressionScalarParameter_17 {class=MaterialExpressionScalarParameter, properties {defaultValue=0.5, group=WPO, parameterName=Frequency, sortPriority=1}}
  MaterialExpressionScalarParameter_18 {class=MaterialExpressionScalarParameter, properties {defaultValue=1, group=Group Hypno, parameterName=EdgeFade}}
  MaterialExpressionScalarParameter_19 {class=MaterialExpressionScalarParameter, properties {defaultValue=1, group=Group Hypno, parameterName=Detail Pattenr Core?}}
  MaterialExpressionScalarParameter_2 {class=MaterialExpressionScalarParameter, properties {defaultValue=0.775, group=Group Swirl, parameterName=Ciclre Radius}}
  MaterialExpressionScalarParameter_20 {class=MaterialExpressionScalarParameter, properties {defaultValue=1.5, group=Group Hypno, parameterName=SwirlTileHypno}}
  MaterialExpressionScalarParameter_21 {class=MaterialExpressionScalarParameter, properties {defaultValue=1, group=Group Hypno, parameterName=contrast Hypno}}
  MaterialExpressionScalarParameter_3 {
   class=MaterialExpressionScalarParameter
   properties {
    controlType=Enumeration
    enumeration=&Enum_DLR_TrueFalse
    group=WPO
    parameterName=Use World Position Offset?
    sortPriority=0
   }
  }
  MaterialExpressionScalarParameter_4 {class=MaterialExpressionScalarParameter, properties {defaultValue=5, parameterName=SphereMaskBrightness}}
  MaterialExpressionScalarParameter_5 {class=MaterialExpressionScalarParameter, properties {defaultValue=5, parameterName=Final Mask Brightness}}
  MaterialExpressionScalarParameter_6 {class=MaterialExpressionScalarParameter, properties {defaultValue=0.5, group=Group Swirl, parameterName=RotationMultiplier}}
  MaterialExpressionScalarParameter_7 {class=MaterialExpressionScalarParameter, properties {defaultValue=1.5, group=Group Swirl, parameterName=SwirlTile}}
  MaterialExpressionScalarParameter_8 {class=MaterialExpressionScalarParameter, properties {defaultValue=1, group=Group Swirl, parameterName=EdgeFadeSwirl, sliderMax=1}}
  MaterialExpressionScalarParameter_9 {class=MaterialExpressionScalarParameter, properties {defaultValue=1, parameterName=Brightness Color}}
  MaterialExpressionSine_1 {class=MaterialExpressionSine, inputs [None {from=MaterialExpressionAdd_8}], properties {period=0.675}}
  MaterialExpressionSine_4 {class=MaterialExpressionSine, inputs [None {from=MaterialExpressionAdd_26}]}
  MaterialExpressionSine_5 {class=MaterialExpressionSine, inputs [None {from=MaterialExpressionMultiply_76}], properties {period=3}}
  MaterialExpressionSine_6 {class=MaterialExpressionSine, inputs [None {from=MaterialExpressionAdd_9}], properties {period=0.75}}
  MaterialExpressionSmoothStep_0 {
   class=MaterialExpressionSmoothStep
   inputs [
    Min {from="", output=""}
    Max {from="", output=""}
    Value {from=MaterialExpressionNamedRerouteUsage_10, output=""}
   ]
   properties {constMax=0.5, constMin=0.5}
  }
  MaterialExpressionSubtract_1 {
   class=MaterialExpressionSubtract
   inputs [
    A {from=MaterialExpressionMultiply_93}
    B {from=MaterialExpressionMultiply_99}
   ]
  }
  MaterialExpressionSubtract_10 {
   class=MaterialExpressionSubtract
   inputs [
    A {from=MaterialExpressionMultiply_75}
    B {from=MaterialExpressionMultiply_80}
   ]
  }
  MaterialExpressionSubtract_2 {
   class=MaterialExpressionSubtract
   inputs [
    A {from=MaterialExpressionMultiply_104}
    B {from=MaterialExpressionMultiply_89}
   ]
  }
  MaterialExpressionSubtract_7 {
   class=MaterialExpressionSubtract
   inputs [
    A {from=MaterialExpressionOneMinus_2}
    B {from=MaterialExpressionScalarParameter_2}
   ]
  }
  MaterialExpressionSubtract_8 {
   class=MaterialExpressionSubtract
   inputs [
    A {from=MaterialExpressionNamedRerouteUsage_28}
    B {from=MaterialExpressionNamedRerouteUsage_42}
   ]
   properties {constB=0.4}
  }
  MaterialExpressionSubtract_9 {
   class=MaterialExpressionSubtract
   inputs [
    A {from=MaterialExpressionConstantClamp_34}
    B {from=MaterialExpressionNamedRerouteUsage_1}
   ]
  }
  MaterialExpressionTextureCoordinate_10 {class=MaterialExpressionTextureCoordinate}
  MaterialExpressionTextureCoordinate_12 {class=MaterialExpressionTextureCoordinate}
  MaterialExpressionTextureCoordinate_13 {class=MaterialExpressionTextureCoordinate, properties {uTiling=-1}}
  MaterialExpressionTextureCoordinate_14 {class=MaterialExpressionTextureCoordinate}
  MaterialExpressionTextureCoordinate_7 {class=MaterialExpressionTextureCoordinate}
  MaterialExpressionTextureSampleParameter2D_0 {
   class=MaterialExpressionTextureSampleParameter2D
   inputs [
    UVs {from=MaterialExpressionPanner_7, output=""}
    Apply View MipBias {from="", output=""}
   ]
   properties {group=Group Hypno, parameterName=PortalRingTexture, texture=&/T_UVDistort_02_RGB}
  }
  MaterialExpressionTextureSampleParameter2D_2 {
   class=MaterialExpressionTextureSampleParameter2D
   inputs [
    UVs {from=MaterialExpressionRotator_8, output=""}
    Apply View MipBias {from="", output=""}
   ]
   properties {group=Group Swirl, parameterName=Texture Swirl, texture=&/T_Swirl_N}
  }
  MaterialExpressionTextureSampleParameter2D_3 {
   class=MaterialExpressionTextureSampleParameter2D
   inputs [
    UVs {from=MaterialExpressionRotator_7, output=""}
    Apply View MipBias {from="", output=""}
   ]
   properties {group=Group Swirl, parameterName=Texture Swirl2, texture=&T_Swirl_N}
  }
  MaterialExpressionTextureSampleParameter2D_4 {
   class=MaterialExpressionTextureSampleParameter2D
   inputs [
    UVs {from=MaterialExpressionMultiply_70, output=""}
    Apply View MipBias {from="", output=""}
   ]
   properties {group=Group Hypno, parameterName=Detail Texture, texture=&/T_Distort_RGB}
  }
  MaterialExpressionTransform_0 {
   class=MaterialExpressionTransform
   inputs [None {from=MaterialExpressionCameraVector_2}]
   properties {transformSourceType=TRANSFORMSOURCE_World, transformType=TRANSFORM_Tangent}
  }
  MaterialExpressionTransform_2 {class=MaterialExpressionTransform, inputs [None {from=MaterialExpressionMultiply_3}]}
  MaterialExpressionVectorParameter_0 {class=MaterialExpressionVectorParameter, properties {group=Group Hypno, parameterName=SpiralDarknessColor}}
 }
 properties {
  bAutomaticallySetUsageInEditor=false
  bUsedWithNiagaraMeshParticles=true
  bUsedWithNiagaraSprites=true
  blendMode=BLEND_Translucent
  shadingModel=MSM_Unlit
  usedShadingModels=Unlit
 }
}
